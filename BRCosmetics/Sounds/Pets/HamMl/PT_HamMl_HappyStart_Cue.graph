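# Sound cue for the HamMl pet's happy-start vocalization, using the pets sound class and a close-range pet attenuation.
# A quality-level node routes its first two levels to a random node that picks among six equally weighted Happy waves.
# Its last two levels skip randomization and play only the first Happy wave.
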SoundCue /BRCosmetics/Sounds/Pets/HamMl/PT_HamMl_HappyStart_Cue
root=QualityLevel_0
nodes [
 {id=QualityLevel_0, children [Random_0,Random_0,WavePlayer_0,WavePlayer_0]}
 {id=Random_0, children [WavePlayer_0,WavePlayer_1,WavePlayer_2,WavePlayer_3,WavePlayer_4,WavePlayer_9], properties {weights [1,1,1,1,1,1]}}
 {id=WavePlayer_0, properties {soundWaveAssetPtr=&/PT_HamMl_Happy_01}}
 {id=WavePlayer_1, properties {soundWaveAssetPtr=&/PT_HamMl_Happy_02}}
 {id=WavePlayer_2, properties {soundWaveAssetPtr=&/PT_HamMl_Happy_03}}
 {id=WavePlayer_3, properties {soundWaveAssetPtr=&/PT_HamMl_Happy_04}}
 {id=WavePlayer_4, properties {soundWaveAssetPtr=&/PT_HamMl_Happy_05}}
 {id=WavePlayer_9, properties {soundWaveAssetPtr=&/PT_HamMl_Happy_06}}
]
properties {
 attenuationSettings=&Fort_Pets_Close_Athena_Att
 maxDistance=1800.0001
 soundClassObject=&Fort_Creatures_Pets
 volumeMultiplier=0.896
}
