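# Sound cue for the HamMl pet's skydiving sound, routed to the creature pets sound class with close-range pet attenuation.
# A quality-level switch feeds the first two levels an equally weighted random pick of five skydive waves, while the other two play only the first wave.
# Overridden concurrency limits it to one instance per owner and blocks new plays while one is active.
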SoundCue /BRCosmetics/Sounds/Pets/HamMl/PT_HamMl_Skydiving_Cue
root=QualityLevel_0
nodes [
 {id=QualityLevel_0, children [Random_0,Random_0,WavePlayer_0,WavePlayer_0]}
 {id=Random_0, children [WavePlayer_0,WavePlayer_1,WavePlayer_3,WavePlayer_5,WavePlayer_6], properties {weights [1,1,1,1,1]}}
 {id=WavePlayer_0, properties {soundWaveAssetPtr=&/PT_HamMl_SkyDive_01}}
 {id=WavePlayer_1, properties {soundWaveAssetPtr=&/PT_HamMl_SkyDive_02}}
 {id=WavePlayer_3, properties {soundWaveAssetPtr=&/PT_HamMl_SkyDive_03}}
 {id=WavePlayer_5, properties {soundWaveAssetPtr=&/PT_HamMl_SkyDive_04}}
 {id=WavePlayer_6, properties {soundWaveAssetPtr=&/PT_HamMl_SkyDive_05}}
]
properties {
 attenuationSettings=&Fort_Pets_Close_Athena_Att
 bOverrideConcurrency=true
 maxDistance=1800.0001
 soundClassObject=&Fort_Creatures_Pets
 volumeMultiplier=0.448
 concurrencyOverrides {bLimitToOwner=true, maxCount=1, resolutionRule=PreventNew}
}
